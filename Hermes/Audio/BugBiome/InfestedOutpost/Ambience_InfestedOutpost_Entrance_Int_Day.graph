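# Looping, non-spatialized daytime interior ambience cue for the Infested Outpost entrance, on an indoor ambient sound class.
# Each mix layers a pitched-down room tone loop with a quieter distant fly-swarm loop whose volume swells cyclically.
# A quality switch gives the first two levels a quad bed with random swarm variants, the others a stereo bed with one swarm.
# Nested looping envelopers of differing lengths add slow pitch and volume drift to the whole mix.
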SoundCue /Hermes/Audio/BugBiome/InfestedOutpost/Ambience_InfestedOutpost_Entrance_Int_Day
root=Enveloper_3
nodes [
 {
  id=Enveloper_0
  children [QualityLevel_1]
  properties {
   bLoop=true
   bLoopIndefinitely=true
   loopEnd=30
   pitchCurve {editorCurveData {keys [{interpMode=RCIM_Cubic, time=0}, <at 15.9468 0.9181 RCIM_Cubic>, {interpMode=RCIM_Cubic, time=30}]}}
   volumeCurve {editorCurveData {keys [<at 0 1>]}}
  }
 }
 {
  id=Enveloper_1
  children [Random_0]
  properties {
   bLoop=true
   bLoopIndefinitely=true
   loopEnd=30
   pitchCurve {editorCurveData {keys [{time=0}]}}
   volumeCurve {editorCurveData {keys [<at 0 1 RCIM_Cubic>,<at 15 0.3 RCIM_Cubic>,<at 30 1 RCIM_Cubic>]}}
  }
 }
 {
  id=Enveloper_2
  children [WavePlayer_5]
  properties {
   bLoop=true
   bLoopIndefinitely=true
   loopEnd=30
   pitchCurve {editorCurveData {keys [{time=0}]}}
   volumeCurve {editorCurveData {keys [<at 0 1 RCIM_Cubic>,<at 15 0.3 RCIM_Cubic>,<at 30 1 RCIM_Cubic>]}}
  }
 }
 {
  id=Enveloper_3
  children [Enveloper_0]
  properties {
   bLoop=true
   bLoopIndefinitely=true
   loopEnd=20
   pitchCurve {editorCurveData {keys [{time=0}]}}
   volumeCurve {editorCurveData {keys [
      {arriveTangent=-0.01, interpMode=RCIM_Cubic, leaveTangent=-0.01, time=0, value=1}
      <at 10 0.9 RCIM_Cubic>
      {arriveTangent=0.01, interpMode=RCIM_Cubic, leaveTangent=0.01, time=20, value=1}
     ]}}
  }
 }
 {id=Mixer_0, children [Modulator_0,Enveloper_1], properties {inputVolume [1,0.2]}}
 {id=Mixer_1, children [Modulator_1,Enveloper_2], properties {inputVolume [1,0.2]}}
 {id=Modulator_0, children [WavePlayer_0], properties {pitchMax=0.8, pitchMin=0.75}}
 {id=Modulator_1, children [WavePlayer_1], properties {pitchMax=0.8, pitchMin=0.75}}
 {id=QualityLevel_1, children [Mixer_0,Mixer_0,Mixer_1,Mixer_1]}
 {id=Random_0, children [WavePlayer_2,WavePlayer_3,WavePlayer_4], properties {weights [1,1,1]}}
 {
  id=WavePlayer_0
  properties {bLooping=true, soundWaveAssetPtr=&/Ambience_TheRaft_Small_Room_01_Int_Quad}
 }
 {
  id=WavePlayer_1
  properties {bLooping=true, soundWaveAssetPtr=&/Ambience_TheRaft_Small_Room_01_Int_Stereo}
 }
 {id=WavePlayer_2, properties {bLooping=true, soundWaveAssetPtr=&/BossSpire_FlySwarm_Distant_Loop_01}}
 {id=WavePlayer_3, properties {bLooping=true, soundWaveAssetPtr=&/BossSpire_FlySwarm_Distant_Loop_02}}
 {id=WavePlayer_4, properties {bLooping=true, soundWaveAssetPtr=&/BossSpire_FlySwarm_Distant_Loop_03}}
 {id=WavePlayer_5, properties {bLooping=true, soundWaveAssetPtr=&BossSpire_FlySwarm_Distant_Loop_01}}
]
properties {
 bOverrideConcurrency=true
 maxDistance=2097152
 soundClassObject=&Fort_Ambient_Indoors_Athena
 volumeMultiplier=1
 attenuationOverrides {bAttenuate=false, bEnableReverbSend=false, bSpatialize=false}
 concurrencyOverrides {resolutionRule=PreventNew}
}
