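# Non-spatialized looping indoor ambience cue for the Infested Outpost stairs at night, in the Bug Biome.
# A quality-level switch plays either a quad canteen bed with a random one of three distant fly-swarm loops, or a stereo bed with a single swarm loop.
# The quietly mixed swarms pulse through looping volume envelopes.
# Two outer envelopes of differing lengths add slow pitch drift and subtle volume movement to vary the loops.
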SoundCue /Hermes/Audio/BugBiome/InfestedOutpost/Ambience_InfestedOutpost_Stairs_Int_Night
root=Enveloper_3
nodes [
 {
  id=Enveloper_0
  children [QualityLevel_0]
  properties {
   bLoop=true
   bLoopIndefinitely=true
   loopEnd=50
   pitchCurve {editorCurveData {keys [{interpMode=RCIM_Cubic, time=0}, <at 25 0.9181 RCIM_Cubic>, {interpMode=RCIM_Cubic, time=49}]}}
   volumeCurve {editorCurveData {keys [<at 0 1>]}}
  }
 }
 {
  id=Enveloper_1
  children [WavePlayer_11]
  properties {
   bLoop=true
   bLoopIndefinitely=true
   loopEnd=30
   pitchCurve {editorCurveData {keys [{time=0}]}}
   volumeCurve {editorCurveData {keys [<at 0 1 RCIM_Cubic>,<at 15 0.3 RCIM_Cubic>,<at 30 1 RCIM_Cubic>]}}
  }
 }
 {
  id=Enveloper_2
  children [Random_1]
  properties {
   bLoop=true
   bLoopIndefinitely=true
   loopEnd=30
   pitchCurve {editorCurveData {keys [{time=0}]}}
   volumeCurve {editorCurveData {keys [<at 0 1 RCIM_Cubic>,<at 15 0.3 RCIM_Cubic>,<at 30 1 RCIM_Cubic>]}}
  }
 }
 {
  id=Enveloper_3
  children [Enveloper_0]
  properties {
   bLoop=true
   bLoopIndefinitely=true
   loopEnd=20
   pitchCurve {editorCurveData {keys [{time=0}]}}
   volumeCurve {editorCurveData {keys [
      {arriveTangent=-0.01, interpMode=RCIM_Cubic, leaveTangent=-0.01, time=0, value=1}
      <at 10 0.9 RCIM_Cubic>
      {arriveTangent=0.01, interpMode=RCIM_Cubic, leaveTangent=0.01, time=20, value=1}
     ]}}
  }
 }
 {id=Mixer_2, children [WavePlayer_0,Enveloper_2], properties {inputVolume [1,0.15]}}
 {id=Mixer_3, children [WavePlayer_3,Enveloper_1], properties {inputVolume [1,0.15]}}
 {id=QualityLevel_0, children [Mixer_2,Mixer_2,Mixer_3,Mixer_3]}
 {id=Random_1, children [WavePlayer_8, WavePlayer_9, WavePlayer_10], properties {weights [1,1,1]}}
 {
  id=WavePlayer_0
  properties {bLooping=true, soundWaveAssetPtr=&/Ambience_TheRaft_Canteen_Int_01_Loop_Quad}
 }
 {id=WavePlayer_10, properties {bLooping=true, soundWaveAssetPtr=&/BossSpire_FlySwarm_Distant_Loop_03}}
 {id=WavePlayer_11, properties {bLooping=true, soundWaveAssetPtr=&/BossSpire_FlySwarm_Distant_Loop_01}}
 {
  id=WavePlayer_3
  properties {bLooping=true, soundWaveAssetPtr=&/Ambience_TheRaft_Canteen_Int_01_Loop_Stereo}
 }
 {id=WavePlayer_8, properties {bLooping=true, soundWaveAssetPtr=&BossSpire_FlySwarm_Distant_Loop_01}}
 {id=WavePlayer_9, properties {bLooping=true, soundWaveAssetPtr=&/BossSpire_FlySwarm_Distant_Loop_02}}
]
properties {
 bOverrideConcurrency=true
 maxDistance=2097152
 soundClassObject=&Fort_Ambient_Indoors_Athena
 volumeMultiplier=1
 attenuationOverrides {bAttenuate=false, bEnableReverbSend=false, bSpatialize=false}
 concurrencyOverrides {resolutionRule=PreventNew}
 soundSubmixSends [{minSendLevel=0, soundSubmix=null}]
}
