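# Sound cue for distant metal rattle one-shot spot ambience in the Infested Outpost of the BugBiome, reusing roller coaster rattle recordings.
# A quality switch sends its first two levels through a slight upward pitch modulator into an equal-weight random pick with six waves and two empty slots.
# The last two levels pick from three of those waves without modulation.
# It uses distant-creak attenuation, with bus and submix sends disabled.
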SoundCue /Hermes/Audio/BugBiome/InfestedOutpost/Spot_InfestedOutpost_Metal_Rattle_Distant_Cue
root=QualityLevel_0
nodes [
 {id=Modulator_0, children [Random_0], properties {pitchMax=1.1, pitchMin=1}}
 {id=QualityLevel_0, children [Modulator_0,Modulator_0,Random_2,Random_2]}
 {
  id=Random_0
  children [null,null,WavePlayer_3,WavePlayer_4,WavePlayer_5,WavePlayer_6,WavePlayer_7,WavePlayer_8]
  properties {weights [1,1,1,1,1,1,1,1]}
 }
 {id=Random_2, children [WavePlayer_9, WavePlayer_10, WavePlayer_11], properties {weights [1,1,1]}}
 {
  id=WavePlayer_10
  properties {soundWaveAssetPtr=/Game/Sounds/Artemis/POI/Hollow_Base_Takover/Spots/Spot_Roller_Coaster_Metal_Rattles_Distant_08}
 }
 {
  id=WavePlayer_11
  properties {soundWaveAssetPtr=/Game/Sounds/Artemis/POI/Hollow_Base_Takover/Spots/Spot_Roller_Coaster_Metal_Rattles_Distant_09}
 }
 {
  id=WavePlayer_3
  properties {soundWaveAssetPtr=/Game/Sounds/Artemis/POI/Hollow_Base_Takover/Spots/Spot_Roller_Coaster_Metal_Rattles_Distant_04}
 }
 {
  id=WavePlayer_4
  properties {soundWaveAssetPtr=/Game/Sounds/Artemis/POI/Hollow_Base_Takover/Spots/Spot_Roller_Coaster_Metal_Rattles_Distant_05}
 }
 {
  id=WavePlayer_5
  properties {soundWaveAssetPtr=/Game/Sounds/Artemis/POI/Hollow_Base_Takover/Spots/Spot_Roller_Coaster_Metal_Rattles_Distant_06}
 }
 {
  id=WavePlayer_6
  properties {soundWaveAssetPtr=/Game/Sounds/Artemis/POI/Hollow_Base_Takover/Spots/Spot_Roller_Coaster_Metal_Rattles_Distant_07}
 }
 {
  id=WavePlayer_7
  properties {soundWaveAssetPtr=&Spot_Roller_Coaster_Metal_Rattles_Distant_08}
 }
 {
  id=WavePlayer_8
  properties {soundWaveAssetPtr=&Spot_Roller_Coaster_Metal_Rattles_Distant_09}
 }
 {
  id=WavePlayer_9
  properties {soundWaveAssetPtr=&Spot_Roller_Coaster_Metal_Rattles_Distant_07}
 }
]
properties {
 attenuationSettings=/Game/Sounds/Attenuation/Fort_TileAmbient_OneShot_Distant_Creak_Att
 bEnableBusSends=false
 bEnableSubmixSends=false
 maxDistance=8200
 soundClassObject=/Game/Sounds/FortSoundClassesAndModes/Fort_Ambient_OneShots_Indoors
 volumeMultiplier=0.501
 attenuationOverrides {falloffDistance=6000, stereoSpread=10000, attenuationShapeExtents {x=2000}}
}
