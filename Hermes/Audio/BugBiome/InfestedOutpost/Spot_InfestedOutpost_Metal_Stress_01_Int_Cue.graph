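# Sound cue for an indoor ambient metal stress spot sound in the Infested Outpost of BugBiome, reusing TheRaft's distant metal "wronk" recordings.
# A quality switch sends the first two levels to an equal-weight random pick of eight waves, and the last two to a smaller pool of three.
# Each branch modulates pitch and volume slightly for variation.
# The cue uses the indoor ambient sound class and overrides attenuation.
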SoundCue /Hermes/Audio/BugBiome/InfestedOutpost/Spot_InfestedOutpost_Metal_Stress_01_Int_Cue
root=QualityLevel_0
nodes [
 {id=Modulator_1, children [Random_0], properties {pitchMax=1.1, pitchMin=1, volumeMax=1.1, volumeMin=0.9}}
 {id=Modulator_2, children [Random_2], properties {pitchMax=1.1, volumeMax=1.1, volumeMin=0.9}}
 {id=QualityLevel_0, children [Modulator_1,Modulator_1,Modulator_2,Modulator_2]}
 {
  id=Random_0
  children [WavePlayer_0,WavePlayer_1,WavePlayer_2,WavePlayer_3,WavePlayer_4,WavePlayer_5,WavePlayer_6,WavePlayer_7]
  properties {weights [1,1,1,1,1,1,1,1]}
 }
 {id=Random_2, children [WavePlayer_8, WavePlayer_9, WavePlayer_21], properties {weights [1,1,1]}}
 {id=WavePlayer_0, properties {soundWaveAssetPtr=&/Spot_TheRaft_Metal_Distant_Wronk_08}}
 {id=WavePlayer_1, properties {soundWaveAssetPtr=&/Spot_TheRaft_Metal_Distant_Wronk_07}}
 {id=WavePlayer_2, properties {soundWaveAssetPtr=&/Spot_TheRaft_Metal_Distant_Wronk_06}}
 {id=WavePlayer_21, properties {soundWaveAssetPtr=&/Spot_TheRaft_Metal_Distant_Wronk_03}}
 {id=WavePlayer_3, properties {soundWaveAssetPtr=&/Spot_TheRaft_Metal_Distant_Wronk_05}}
 {id=WavePlayer_4, properties {soundWaveAssetPtr=&/Spot_TheRaft_Metal_Distant_Wronk_04}}
 {id=WavePlayer_5, properties {soundWaveAssetPtr=&Spot_TheRaft_Metal_Distant_Wronk_03}}
 {id=WavePlayer_6, properties {soundWaveAssetPtr=&/Spot_TheRaft_Metal_Distant_Wronk_02}}
 {id=WavePlayer_7, properties {soundWaveAssetPtr=&/Spot_TheRaft_Metal_Distant_Wronk_01}}
 {id=WavePlayer_8, properties {soundWaveAssetPtr=&Spot_TheRaft_Metal_Distant_Wronk_01}}
 {id=WavePlayer_9, properties {soundWaveAssetPtr=&Spot_TheRaft_Metal_Distant_Wronk_02}}
]
properties {
 attenuationSettings=/Game/Sounds/Attenuation/Fort_TileAmbient_OneShot_Att
 bEnableBusSends=false
 bOverrideAttenuation=true
 maxDistance=4000
 soundClassObject=&Fort_Ambient_Indoors_Athena
 volumeMultiplier=0.501
 preEffectBusSends [{}]
}
