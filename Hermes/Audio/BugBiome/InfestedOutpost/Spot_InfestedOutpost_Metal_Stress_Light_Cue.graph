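# Ambient one-shot cue of light metal stress spot sounds for the Infested Outpost area, reusing Tilted Towers recordings.
# A quality switch sends the first two levels through a pitch-widened modulator into a random pick of ten variations.
# The other two levels use an unmodulated random pick of only the first three, trimming variety by quality.
# Playback uses a less-positional ambient attenuation preset at reduced volume.
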SoundCue /Hermes/Audio/BugBiome/InfestedOutpost/Spot_InfestedOutpost_Metal_Stress_Light_Cue
root=QualityLevel_0
nodes [
 {id=Modulator_0, children [Random_2], properties {pitchMin=0.9}}
 {id=QualityLevel_0, children [Modulator_0,Modulator_0,Random_1,Random_1]}
 {id=Random_1, children [WavePlayer_23, WavePlayer_24, WavePlayer_25], properties {weights [1,1,1]}}
 {
  id=Random_2
  children [
   WavePlayer_13
   WavePlayer_14
   WavePlayer_15
   WavePlayer_16
   WavePlayer_17
   WavePlayer_18
   WavePlayer_19
   WavePlayer_20
   WavePlayer_21
   WavePlayer_22
  ]
  properties {weights [1,1,1,1,1,1,1,1,1,1]}
 }
 {id=WavePlayer_13, properties {soundWaveAssetPtr=/Game/Sounds/Artemis/POI/Tilted_Towers/Spots/Spot_Tilted_Towers_Metal_Stress_Light_01}}
 {id=WavePlayer_14, properties {soundWaveAssetPtr=/Game/Sounds/Artemis/POI/Tilted_Towers/Spots/Spot_Tilted_Towers_Metal_Stress_Light_02}}
 {id=WavePlayer_15, properties {soundWaveAssetPtr=/Game/Sounds/Artemis/POI/Tilted_Towers/Spots/Spot_Tilted_Towers_Metal_Stress_Light_03}}
 {id=WavePlayer_16, properties {soundWaveAssetPtr=/Game/Sounds/Artemis/POI/Tilted_Towers/Spots/Spot_Tilted_Towers_Metal_Stress_Light_04}}
 {id=WavePlayer_17, properties {soundWaveAssetPtr=/Game/Sounds/Artemis/POI/Tilted_Towers/Spots/Spot_Tilted_Towers_Metal_Stress_Light_05}}
 {id=WavePlayer_18, properties {soundWaveAssetPtr=/Game/Sounds/Artemis/POI/Tilted_Towers/Spots/Spot_Tilted_Towers_Metal_Stress_Light_06}}
 {id=WavePlayer_19, properties {soundWaveAssetPtr=/Game/Sounds/Artemis/POI/Tilted_Towers/Spots/Spot_Tilted_Towers_Metal_Stress_Light_07}}
 {id=WavePlayer_20, properties {soundWaveAssetPtr=/Game/Sounds/Artemis/POI/Tilted_Towers/Spots/Spot_Tilted_Towers_Metal_Stress_Light_08}}
 {id=WavePlayer_21, properties {soundWaveAssetPtr=/Game/Sounds/Artemis/POI/Tilted_Towers/Spots/Spot_Tilted_Towers_Metal_Stress_Light_09}}
 {id=WavePlayer_22, properties {soundWaveAssetPtr=/Game/Sounds/Artemis/POI/Tilted_Towers/Spots/Spot_Tilted_Towers_Metal_Stress_Light_10}}
 {id=WavePlayer_23, properties {soundWaveAssetPtr=&Spot_Tilted_Towers_Metal_Stress_Light_01}}
 {id=WavePlayer_24, properties {soundWaveAssetPtr=&Spot_Tilted_Towers_Metal_Stress_Light_02}}
 {id=WavePlayer_25, properties {soundWaveAssetPtr=&Spot_Tilted_Towers_Metal_Stress_Light_03}}
]
properties {
 attenuationSettings=/Game/Sounds/Attenuation/Fort_TileAmbient_OneShot_LessPositional_Log_Att
 maxDistance=6200
 soundClassObject=&Fort_Ambient_OneShots
 volumeMultiplier=0.501
 attenuationOverrides {falloffDistance=6000, stereoSpread=10000, attenuationShapeExtents {x=2000}}
}
