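# Sound cue for indoor ambient one-shots of falling cave rock debris in the Infested Outpost of the bug biome.
# A quality-level switch sends its first two levels through a Modulator into a random pick of nine equally weighted waves.
# Its last two levels use a random pick of only the first three waves, with no modulation.
# It references a ceiling-debris attenuation asset and the indoor ambient one-shot sound class.
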SoundCue /Hermes/Audio/BugBiome/InfestedOutpost/Spot_InfestedOutpost_Rock_Debris_Falling_Int_Cue
root=QualityLevel_1
nodes [
 {id=Modulator_0, children [Random_2]}
 {id=QualityLevel_1, children [Modulator_0,Modulator_0,Random_4,Random_4]}
 {
  id=Random_2
  children [WavePlayer_3, WavePlayer_4, WavePlayer_11, WavePlayer_15, WavePlayer_16, WavePlayer_17, WavePlayer_18, WavePlayer_19, WavePlayer_20]
  properties {weights [1,1,1,1,1,1,1,1,1]}
 }
 {id=Random_4, children [WavePlayer_21, WavePlayer_22, WavePlayer_23], properties {weights [1,1,1]}}
 {id=WavePlayer_11, properties {soundWaveAssetPtr=&/Spot_Cave_Rock_Debris_Falling_03}}
 {id=WavePlayer_15, properties {soundWaveAssetPtr=&/Spot_Cave_Rock_Debris_Falling_04}}
 {id=WavePlayer_16, properties {soundWaveAssetPtr=&/Spot_Cave_Rock_Debris_Falling_05}}
 {id=WavePlayer_17, properties {soundWaveAssetPtr=&/Spot_Cave_Rock_Debris_Falling_06}}
 {id=WavePlayer_18, properties {soundWaveAssetPtr=&/Spot_Cave_Rock_Debris_Falling_07}}
 {id=WavePlayer_19, properties {soundWaveAssetPtr=&/Spot_Cave_Rock_Debris_Falling_08}}
 {id=WavePlayer_20, properties {soundWaveAssetPtr=&/Spot_Cave_Rock_Debris_Falling_09}}
 {id=WavePlayer_21, properties {soundWaveAssetPtr=&/Spot_Cave_Rock_Debris_Falling_01}}
 {id=WavePlayer_22, properties {soundWaveAssetPtr=&/Spot_Cave_Rock_Debris_Falling_02}}
 {id=WavePlayer_23, properties {soundWaveAssetPtr=&Spot_Cave_Rock_Debris_Falling_03}}
 {id=WavePlayer_3, properties {soundWaveAssetPtr=&Spot_Cave_Rock_Debris_Falling_01}}
 {id=WavePlayer_4, properties {soundWaveAssetPtr=&Spot_Cave_Rock_Debris_Falling_02}}
]
properties {
 attenuationSettings=/Game/Sounds/Attenuation/Fort_TileAmbient_OneShot_Ceiling_Debris_Att
 bEnableBusSends=false
 bEnableSubmixSends=false
 maxDistance=6750
 soundClassObject=/Game/Sounds/FortSoundClassesAndModes/Fort_Ambient_OneShots_Indoors
 volumeMultiplier=0.753
 attenuationOverrides {bAttenuate=false}
}
