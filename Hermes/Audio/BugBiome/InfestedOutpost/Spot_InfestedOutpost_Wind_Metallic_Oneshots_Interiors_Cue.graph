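# Sound cue for quiet, randomized metallic tonal wind one-shots in Infested Outpost interiors, using the indoor ambient one-shot sound class.
# A quality-level node feeds its first two slots from an eight-variation random pool and its last two from a three-variation pool.
# Each pool passes through its own pitch and volume modulator, with the whole cue pitched down.
# Shared one-shot attenuation and a concurrency override limit playback; the listed reverb send is disabled.
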SoundCue /Hermes/Audio/BugBiome/InfestedOutpost/Spot_InfestedOutpost_Wind_Metallic_Oneshots_Interiors_Cue
root=QualityLevel_0
nodes [
 {id=Modulator_0, children [Random_1], properties {pitchMax=1.1, pitchMin=0.9, volumeMax=0.9, volumeMin=0.8}}
 {id=Modulator_1, children [Random_2], properties {pitchMax=1.1, pitchMin=0.9, volumeMax=1.1, volumeMin=0.9}}
 {id=QualityLevel_0, children [Modulator_1,Modulator_1,Modulator_0,Modulator_0]}
 {id=Random_1, children [WavePlayer_8, WavePlayer_9, WavePlayer_10], properties {weights [1,1,1]}}
 {
  id=Random_2
  children [WavePlayer_0,WavePlayer_1,WavePlayer_2,WavePlayer_3,WavePlayer_4,WavePlayer_5,WavePlayer_6,WavePlayer_7]
  properties {weights [1,1,1,1,1,1,1,1]}
 }
 {id=WavePlayer_0, properties {soundWaveAssetPtr=&/Spot_Olympus_Tonal_Wind_Met_08}}
 {id=WavePlayer_1, properties {soundWaveAssetPtr=&/Spot_Olympus_Tonal_Wind_Met_07}}
 {id=WavePlayer_10, properties {soundWaveAssetPtr=&/Spot_Olympus_Tonal_Wind_Met_01}}
 {id=WavePlayer_2, properties {soundWaveAssetPtr=&/Spot_Olympus_Tonal_Wind_Met_06}}
 {id=WavePlayer_3, properties {soundWaveAssetPtr=&/Spot_Olympus_Tonal_Wind_Met_05}}
 {id=WavePlayer_4, properties {soundWaveAssetPtr=&/Spot_Olympus_Tonal_Wind_Met_04}}
 {id=WavePlayer_5, properties {soundWaveAssetPtr=&/Spot_Olympus_Tonal_Wind_Met_03}}
 {id=WavePlayer_6, properties {soundWaveAssetPtr=&/Spot_Olympus_Tonal_Wind_Met_02}}
 {id=WavePlayer_7, properties {soundWaveAssetPtr=&Spot_Olympus_Tonal_Wind_Met_01}}
 {id=WavePlayer_8, properties {soundWaveAssetPtr=&Spot_Olympus_Tonal_Wind_Met_03}}
 {id=WavePlayer_9, properties {soundWaveAssetPtr=&Spot_Olympus_Tonal_Wind_Met_02}}
]
properties {
 attenuationSettings=/Game/Sounds/Attenuation/Fort_TileAmbient_OneShot_Att
 bEnableSubmixSends=false
 bOverrideConcurrency=true
 maxDistance=8000
 pitchMultiplier=0.5
 soundClassObject=/Game/Sounds/FortSoundClassesAndModes/Fort_Ambient_OneShots_Indoors
 volumeMultiplier=0.2
 concurrencyOverrides {maxCount=3}
 soundSubmixSends [{maxSendLevel=8, minSendLevel=0, sendLevel=8, soundSubmix=&Fortnite_Reverb_Default}]
}
